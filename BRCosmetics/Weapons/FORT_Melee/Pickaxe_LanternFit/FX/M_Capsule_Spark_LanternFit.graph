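# Particle material for LanternFit pickaxe sparks: soft capsule or teardrop sprites tinted by particle color with an HDR glow.
# A sphere mask shapes glow and opacity, with depth and camera fades and a previous-frame switch on the world position offset.
# A lamp "LightsOn" signal multiplies opacity, built from texture ID masks and sine blinking.
# When reactivity is enabled, dynamic-parameter timestamps drive flicker, lights-off and reaction phases, including a separate yellow-light timer.
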Material /BRCosmetics/Weapons/FORT_Melee/Pickaxe_LanternFit/FX/M_Capsule_Spark_LanternFit
graph {
 expressions {
  MaterialExpressionAdd_10 {
   class=MaterialExpressionAdd
   inputs [
    A {from=MaterialExpressionSaturate_41}
    B {from=MaterialExpressionSaturate_42}
   ]
  }
  MaterialExpressionAdd_11 {
   class=MaterialExpressionAdd
   inputs [
    A {from=MaterialExpressionDynamicParameter_6, output=ReactivityTimestamp}
    B {from=MaterialExpressionAdd_12}
   ]
  }
  MaterialExpressionAdd_12 {
   class=MaterialExpressionAdd
   inputs [
    A {from=MaterialExpressionScalarParameter_42}
    B {from=MaterialExpressionConstant_33}
   ]
   properties {constB=3, desc=eXTRAdURATION}
  }
  MaterialExpressionAdd_14 {
   class=MaterialExpressionAdd
   inputs [
    A {from=MaterialExpressionTextureCoordinate_5}
    B {from=MaterialExpressionAppendVector_2}
   ]
   properties {constB=0}
  }
  MaterialExpressionAdd_4 {
   class=MaterialExpressionAdd
   inputs [
    A {from=MaterialExpressionDynamicParameter_7, output=YellowLightsTimestamp}
    B {from=MaterialExpressionAdd_5}
   ]
  }
  MaterialExpressionAdd_5 {
   class=MaterialExpressionAdd
   inputs [
    A {from=MaterialExpressionScalarParameter_46}
    B {from=MaterialExpressionConstant_36}
   ]
   properties {constB=3, desc=eXTRAdURATION}
  }
  MaterialExpressionAdd_9 {
   class=MaterialExpressionAdd
   inputs [
    A {from=MaterialExpressionLinearInterpolate_20}
    B {from=MaterialExpressionMultiply_43}
   ]
  }
  MaterialExpressionAppendVector_2 {
   class=MaterialExpressionAppendVector
   inputs [
    A {from=MaterialExpressionDynamicParameter_4, output=Offset1}
    B {from=MaterialExpressionDynamicParameter_4, output=Offset1}
   ]
  }
  MaterialExpressionCeil_0 {class=MaterialExpressionCeil, inputs [None {from=MaterialExpressionMultiply_58, output=""}], properties {}}
  MaterialExpressionCeil_2 {class=MaterialExpressionCeil, inputs [None {from=MaterialExpressionMultiply_46, output=""}], properties {}}
  MaterialExpressionClamp_0 {
   class=MaterialExpressionClamp
   inputs [
    None {from=MaterialExpressionMultiply_0, output=""}
    Min {from="", output=""}
    Max {from="", output=""}
   ]
   properties {}
  }
  MaterialExpressionComponentMask_0 {class=MaterialExpressionComponentMask, inputs [None {from=MaterialExpressionTextureCoordinate_0}], properties {r=true}}
  MaterialExpressionComponentMask_4 {class=MaterialExpressionComponentMask, inputs [None {from=MaterialExpressionTextureCoordinate_4}], properties {r=true}}
  MaterialExpressionConstantBiasScale_1 {class=MaterialExpressionConstantBiasScale, inputs [None {from=MaterialExpressionNamedRerouteUsage_73}], properties {bias=-0.5, scale=2}}
  MaterialExpressionConstant_0 {class=MaterialExpressionConstant, properties {r=0.47}}
  MaterialExpressionConstant_1 {class=MaterialExpressionConstant, properties {r=0.5}}
  MaterialExpressionConstant_25 {class=MaterialExpressionConstant}
  MaterialExpressionConstant_26 {class=MaterialExpressionConstant, properties {r=1}}
  MaterialExpressionConstant_27 {class=MaterialExpressionConstant}
  MaterialExpressionConstant_28 {class=MaterialExpressionConstant}
  MaterialExpressionConstant_29 {class=MaterialExpressionConstant, properties {r=1}}
  MaterialExpressionConstant_30 {class=MaterialExpressionConstant}
  MaterialExpressionConstant_31 {class=MaterialExpressionConstant, properties {r=1}}
  MaterialExpressionConstant_32 {class=MaterialExpressionConstant, properties {r=1}}
  MaterialExpressionConstant_33 {class=MaterialExpressionConstant, properties {r=0.001}}
  MaterialExpressionConstant_34 {class=MaterialExpressionConstant}
  MaterialExpressionConstant_35 {class=MaterialExpressionConstant, properties {r=1}}
  MaterialExpressionConstant_36 {class=MaterialExpressionConstant, properties {r=0.001}}
  MaterialExpressionCosine_2 {class=MaterialExpressionCosine, inputs [None {from=MaterialExpressionAdd_9, output=""}], properties {period=1.3457}}
  MaterialExpressionDeltaTime_0 {class=MaterialExpressionDeltaTime, inputs [], properties {}}
  MaterialExpressionDepthFade_0 {
   class=MaterialExpressionDepthFade
   inputs [
    Opacity {from=MaterialExpressionReroute_0, output=""}
    FadeDistance {from=MaterialExpressionDynamicParameter_0, output=DepthFade}
   ]
   properties {}
  }
  MaterialExpressionDivide_12 {
   class=MaterialExpressionDivide
   inputs [
    A {from=MaterialExpressionMultiply_50}
    B {from=""}
   ]
   properties {constB=4}
  }
  MaterialExpressionDivide_13 {
   class=MaterialExpressionDivide
   inputs [
    A {from=MaterialExpressionSubtract_19}
    B {from=MaterialExpressionScalarParameter_42}
   ]
   properties {desc=InLength}
  }
  MaterialExpressionDivide_14 {
   class=MaterialExpressionDivide
   inputs [
    A {from=MaterialExpressionSubtract_20}
    B {from=""}
   ]
   properties {constB=0.001, desc=OutLength}
  }
  MaterialExpressionDivide_15 {
   class=MaterialExpressionDivide
   inputs [
    A {from=MaterialExpressionTime_11}
    B {from=MaterialExpressionScalarParameter_42}
   ]
   properties {constB=20}
  }
  MaterialExpressionDivide_16 {
   class=MaterialExpressionDivide
   inputs [
    A {from=MaterialExpressionDivide_15}
    B {from=""}
   ]
   properties {constB=0.5}
  }
  MaterialExpressionDivide_7 {
   class=MaterialExpressionDivide
   inputs [
    A {from=MaterialExpressionSubtract_8}
    B {from=MaterialExpressionScalarParameter_46}
   ]
   properties {desc=InLength}
  }
  MaterialExpressionDivide_8 {
   class=MaterialExpressionDivide
   inputs [
    A {from=MaterialExpressionSubtract_9}
    B {from=""}
   ]
   properties {constB=0.001, desc=OutLength}
  }
  MaterialExpressionDynamicParameter_0 {
   class=MaterialExpressionDynamicParameter
   inputs []
   properties {defaultValue=<rgba 2 10 1 1>, paramNames [Power,DepthFade,HDRMult,BottomFade]}
  }
  MaterialExpressionDynamicParameter_1 {
   class=MaterialExpressionDynamicParameter
   inputs []
   properties {parameterIndex=1, defaultValue=<rgba 20 500 10 1>, paramNames [TeardropMinSpeed, TeardropMaxSpeed, TeardropMaxStretch, Param4]}
  }
  MaterialExpressionDynamicParameter_4 {
   class=MaterialExpressionDynamicParameter
   inputs []
   properties {parameterIndex=2, defaultValue=<rgba 1 1 0 -999>, paramNames [Offset1, Offset2, ReactivityID, ReactivityTimestamp]}
  }
  MaterialExpressionDynamicParameter_5 {
   class=MaterialExpressionDynamicParameter
   inputs []
   properties {parameterIndex=2, defaultValue=<rgba 1 1 0 -999>, paramNames [Offset1, Offset2, ReactivityID, ReactivityTimestamp]}
  }
  MaterialExpressionDynamicParameter_6 {
   class=MaterialExpressionDynamicParameter
   inputs []
   properties {parameterIndex=2, defaultValue=<rgba 1 1 0 -999>, paramNames [Offset1, Offset2, ReactivityID, ReactivityTimestamp]}
  }
  MaterialExpressionDynamicParameter_7 {
   class=MaterialExpressionDynamicParameter
   inputs []
   properties {parameterIndex=3, defaultValue=<rgba -999 1 1 1>, paramNames [YellowLightsTimestamp, Param2, Param3, Param4]}
  }
  MaterialExpressionFrac_10 {class=MaterialExpressionFrac, inputs [None {from=MaterialExpressionMultiply_52, output=""}], properties {}}
  MaterialExpressionFrac_11 {class=MaterialExpressionFrac, inputs [None {from=MaterialExpressionMultiply_54, output=""}], properties {}}
  MaterialExpressionFrac_8 {class=MaterialExpressionFrac, inputs [None {from=MaterialExpressionDivide_12, output=""}], properties {}}
  MaterialExpressionFrac_9 {class=MaterialExpressionFrac, inputs [None {from=MaterialExpressionMultiply_58, output=""}], properties {}}
  MaterialExpressionIf_2 {
   class=MaterialExpressionIf
   inputs [
    A {from=MaterialExpressionNamedRerouteUsage_77, output=""}
    B {from="", output=""}
    "A > B" {from=MaterialExpressionStep_39, output=""}
    "A == B" {from=MaterialExpressionStep_38, output=""}
    "A < B" {from=MaterialExpressionStep_38, output=""}
   ]
   properties {constB=1}
  }
  MaterialExpressionLinearInterpolate_0 {
   class=MaterialExpressionLinearInterpolate
   inputs [
    A {from=MaterialExpressionScalarParameter_2}
    B {from=MaterialExpressionScalarParameter_1}
    Alpha {from=MaterialExpressionPower_0}
   ]
  }
  MaterialExpressionLinearInterpolate_1 {
   class=MaterialExpressionLinearInterpolate
   inputs [
    A {from=MaterialExpressionDynamicParameter_0, output=BottomFade}
    B {from=""}
    Alpha {from=MaterialExpressionComponentMask_0}
   ]
   properties {constA=1}
  }
  MaterialExpressionLinearInterpolate_17 {
   class=MaterialExpressionLinearInterpolate
   inputs [
    A {from=MaterialExpressionScalarParameter_36}
    B {from=MaterialExpressionScalarParameter_39}
    Alpha {from=MaterialExpressionStep_42}
   ]
  }
  MaterialExpressionLinearInterpolate_18 {
   class=MaterialExpressionLinearInterpolate
   inputs [
    A {from=MaterialExpressionMultiply_47}
    B {from=MaterialExpressionNamedRerouteUsage_76}
    Alpha {from=MaterialExpressionStep_44}
   ]
  }
  MaterialExpressionLinearInterpolate_19 {
   class=MaterialExpressionLinearInterpolate
   inputs [
    A {from=MaterialExpressionMultiply_58}
    B {from=MaterialExpressionReroute_30}
    Alpha {from=MaterialExpressionNamedRerouteDeclaration_62}
   ]
  }
  MaterialExpressionLinearInterpolate_20 {
   class=MaterialExpressionLinearInterpolate
   inputs [
    A {from=MaterialExpressionReroute_28}
    B {from=MaterialExpressionMultiply_51}
    Alpha {from=MaterialExpressionStep_43}
   ]
  }
  MaterialExpressionLinearInterpolate_21 {
   class=MaterialExpressionLinearInterpolate
   inputs [
    A {from=MaterialExpressionStep_40}
    B {from=MaterialExpressionSine_15}
    Alpha {from=MaterialExpressionStep_41}
   ]
  }
  MaterialExpressionLinearInterpolate_22 {
   class=MaterialExpressionLinearInterpolate
   inputs [
    A {from=MaterialExpressionLinearInterpolate_18}
    B {from=MaterialExpressionMultiply_48}
    Alpha {from=MaterialExpressionNamedRerouteUsage_10}
   ]
  }
  MaterialExpressionMaterialFunctionCall_0 {
   class=MaterialExpressionMaterialFunctionCall
   inputs [
    LDR Input {from=MaterialExpressionClamp_0}
    HDR Dynamic Range {from=""}
    HDR Tint {from=MaterialExpressionMultiply_0}
    HDR Intensity Multiplier {from=MaterialExpressionReroute_2}
    Derive HDR Power {from=""}
    Desaturate Luminance {from=""}
    Derive Luminance from Linear {from=""}
   ]
   properties {materialFunction=&DeriveHDRfromLDR}
  }
  MaterialExpressionMaterialFunctionCall_1 {
   class=MaterialExpressionMaterialFunctionCall
   inputs [Emissive {from=MaterialExpressionShadingPathSwitch_1}]
   properties {materialFunction=&MF_FX_EmissiveAdjustments}
  }
  MaterialExpressionMaterialFunctionCall_14 {
   class=MaterialExpressionMaterialFunctionCall
   inputs [
    Input {from=MaterialExpressionReroute_27}
    Input Low {from=MaterialExpressionConstant_27}
    Input High {from=MaterialExpressionScalarParameter_41}
    Target Low {from=MaterialExpressionConstant_25}
    Target High {from=MaterialExpressionConstant_26}
   ]
   properties {materialFunction=&RemapValueRange}
  }
  MaterialExpressionMaterialFunctionCall_15 {
   class=MaterialExpressionMaterialFunctionCall
   inputs [
    Input {from=MaterialExpressionReroute_23}
    Input Low {from=MaterialExpressionScalarParameter_41}
    Input High {from=MaterialExpressionScalarParameter_40}
    Target Low {from=MaterialExpressionConstant_28}
    Target High {from=MaterialExpressionConstant_29}
   ]
   properties {materialFunction=&RemapValueRange}
  }
  MaterialExpressionMaterialFunctionCall_16 {
   class=MaterialExpressionMaterialFunctionCall
   inputs [
    Input {from=MaterialExpressionReroute_24}
    Input Low {from=MaterialExpressionScalarParameter_43}
    Input High {from=MaterialExpressionConstant_32}
    Target Low {from=MaterialExpressionConstant_30}
    Target High {from=MaterialExpressionConstant_31}
   ]
   properties {materialFunction=&RemapValueRange}
  }
  MaterialExpressionMaterialFunctionCall_17 {
   class=MaterialExpressionMaterialFunctionCall
   inputs [
    Input {from=MaterialExpressionReroute_29}
    Input Low {from=MaterialExpressionScalarParameter_40}
    Input High {from=MaterialExpressionScalarParameter_43}
    Target Low {from=MaterialExpressionConstant_34}
    Target High {from=MaterialExpressionConstant_35}
   ]
   properties {materialFunction=&RemapValueRange}
  }
  MaterialExpressionMaterialFunctionCall_2 {
   class=MaterialExpressionMaterialFunctionCall
   inputs [
    Max Stretch Factor {from=MaterialExpressionDynamicParameter_1, output=TeardropMaxStretch}
    Min Tear Drop Width {from=MaterialExpressionScalarParameter_3}
    Minimum Speed {from=MaterialExpressionDynamicParameter_1, output=TeardropMaxStretch}
    Max Speed {from=MaterialExpressionDynamicParameter_1, output=TeardropMaxStretch}
    ------ Particle Defaults Below ------ {from=""}
    Particle Speed {from=""}
    Particle Direction {from=""}
    Particle Position {from=""}
    Particle Size {from=""}
    UV Add {from=""}
   ]
   properties {materialFunction=/Engine/Functions/Engine_MaterialFunctions02/WorldPositionOffset/Sprite_TearDrop}
  }
  MaterialExpressionMaterialFunctionCall_4 {
   class=MaterialExpressionMaterialFunctionCall
   inputs [
    Particle Position {from=""}
    Normalized Particle Direction {from=""}
    Capsule Diameter {from=""}
    Capsule Pivot Alignment {from=""}
    Scale Along Velocity Vector {from=""}
    ---------Scale Over Velocity Parameters--------- {from=""}
    Particle Speed {from=""}
    Minimum Speed {from=""}
    Max Speed {from=""}
    Minimum Scale Factor {from=MaterialExpressionMaterialFunctionCall_5, output=G}
    Maximum Scale Factor {from=""}
    ---------- Advanced ----------- {from=""}
    Cap Size Multiplier {from=MaterialExpressionScalarParameter_0}
    Use Camera Offset {from=MaterialExpressionStaticBool_0}
    Use Advanced Features {from=MaterialExpressionStaticBool_0}
    World Position to Negate {from=""}
   ]
   properties {materialFunction=/Engine/Functions/Engine_MaterialFunctions02/WorldPositionOffset/Sprite_Capsule}
  }
  MaterialExpressionMaterialFunctionCall_5 {
   class=MaterialExpressionMaterialFunctionCall
   inputs [Float2 {from=MaterialExpressionParticleSize_1}]
   properties {materialFunction=&BreakOutFloat2Components}
  }
  MaterialExpressionMaterialFunctionCall_6 {
   class=MaterialExpressionMaterialFunctionCall
   inputs [
    Fade Length {from=MaterialExpressionScalarParameter_5}
    Fade Offset {from=MaterialExpressionScalarParameter_6}
    For Vertex Shader {from=""}
   ]
   properties {materialFunction=&CameraDepthFade}
  }
  MaterialExpressionMax_5 {
   class=MaterialExpressionMax
   inputs [
    A {from=MaterialExpressionNamedRerouteUsage_60}
    B {from=MaterialExpressionNamedRerouteUsage_61}
   ]
  }
  MaterialExpressionMax_6 {
   class=MaterialExpressionMax
   inputs [
    A {from=MaterialExpressionNamedRerouteUsage_64}
    B {from=MaterialExpressionNamedRerouteUsage_65}
   ]
  }
  MaterialExpressionMultiply_0 {
   class=MaterialExpressionMultiply
   inputs [
    A {from=MaterialExpressionParticleColor_0, output=RGB}
    B {from=MaterialExpressionLinearInterpolate_0}
   ]
  }
  MaterialExpressionMultiply_1 {
   class=MaterialExpressionMultiply
   inputs [
    A {from=MaterialExpressionParticleColor_0, output=A}
    B {from=MaterialExpressionShadingPathSwitch_0}
   ]
  }
  MaterialExpressionMultiply_2 {
   class=MaterialExpressionMultiply
   inputs [
    A {from=MaterialExpressionParticleDirection_0}
    B {from=MaterialExpressionParticleSpeed_0}
   ]
  }
  MaterialExpressionMultiply_3 {
   class=MaterialExpressionMultiply
   inputs [
    A {from=MaterialExpressionMultiply_2}
    B {from=MaterialExpressionDeltaTime_0}
   ]
  }
  MaterialExpressionMultiply_4 {
   class=MaterialExpressionMultiply
   inputs [
    A {from=MaterialExpressionMultiply_3}
    B {from=""}
   ]
  }
  MaterialExpressionMultiply_43 {
   class=MaterialExpressionMultiply
   inputs [
    A {from=MaterialExpressionTime_9}
    B {from=MaterialExpressionStaticSwitchParameter_8}
   ]
  }
  MaterialExpressionMultiply_44 {
   class=MaterialExpressionMultiply
   inputs [
    A {from=MaterialExpressionFrac_8}
    B {from=""}
   ]
   properties {constB=4}
  }
  MaterialExpressionMultiply_45 {
   class=MaterialExpressionMultiply
   inputs [
    A {from=MaterialExpressionSaturate_43}
    B {from=MaterialExpressionSaturate_44}
   ]
  }
  MaterialExpressionMultiply_46 {
   class=MaterialExpressionMultiply
   inputs [
    A {from=MaterialExpressionSaturate_45}
    B {from=MaterialExpressionOneMinus_6}
   ]
  }
  MaterialExpressionMultiply_47 {
   class=MaterialExpressionMultiply
   inputs [
    A {from=MaterialExpressionSmoothStep_1}
    B {from=MaterialExpressionOneMinus_8}
   ]
  }
  MaterialExpressionMultiply_48 {
   class=MaterialExpressionMultiply
   inputs [
    A {from=MaterialExpressionSine_14}
    B {from=MaterialExpressionNamedRerouteUsage_75}
   ]
  }
  MaterialExpressionMultiply_49 {
   class=MaterialExpressionMultiply
   inputs [
    A {from=MaterialExpressionNamedRerouteDeclaration_60}
    B {from=MaterialExpressionReroute_26}
   ]
  }
  MaterialExpressionMultiply_5 {
   class=MaterialExpressionMultiply
   inputs [
    A {from=MaterialExpressionLinearInterpolate_1}
    B {from=MaterialExpressionSphereMask_0}
   ]
  }
  MaterialExpressionMultiply_50 {
   class=MaterialExpressionMultiply
   inputs [
    A {from=MaterialExpressionNamedRerouteUsage_81}
    B {from=MaterialExpressionScalarParameter_37}
   ]
   properties {constB=25}
  }
  MaterialExpressionMultiply_51 {
   class=MaterialExpressionMultiply
   inputs [
    A {from=MaterialExpressionConstantBiasScale_1}
    B {from=""}
   ]
   properties {constB=0.25}
  }
  MaterialExpressionMultiply_52 {
   class=MaterialExpressionMultiply
   inputs [
    A {from=MaterialExpressionTime_12}
    B {from=""}
   ]
   properties {constB=0.3}
  }
  MaterialExpressionMultiply_53 {
   class=MaterialExpressionMultiply
   inputs [
    A {from=MaterialExpressionTime_12}
    B {from=""}
   ]
   properties {constB=0.5}
  }
  MaterialExpressionMultiply_54 {
   class=MaterialExpressionMultiply
   inputs [
    A {from=MaterialExpressionMultiply_53}
    B {from=""}
   ]
   properties {constB=0.3}
  }
  MaterialExpressionMultiply_55 {
   class=MaterialExpressionMultiply
   inputs [
    A {from=MaterialExpressionReroute_25}
    B {from=MaterialExpressionScalarParameter_44}
   ]
   properties {constB=3}
  }
  MaterialExpressionMultiply_56 {
   class=MaterialExpressionMultiply
   inputs [
    A {from=MaterialExpressionDivide_15}
    B {from=""}
   ]
   properties {constB=0.25}
  }
  MaterialExpressionMultiply_57 {
   class=MaterialExpressionMultiply
   inputs [
    A {from=MaterialExpressionStep_8}
    B {from=MaterialExpressionStep_45}
   ]
  }
  MaterialExpressionMultiply_58 {
   class=MaterialExpressionMultiply
   inputs [
    A {from=MaterialExpressionSaturate_18}
    B {from=MaterialExpressionOneMinus_4}
   ]
  }
  MaterialExpressionMultiply_6 {
   class=MaterialExpressionMultiply
   inputs [
    A {from=MaterialExpressionMultiply_1}
    B {from=MaterialExpressionMaterialFunctionCall_6, output=Result}
   ]
  }
  MaterialExpressionMultiply_8 {
   class=MaterialExpressionMultiply
   inputs [
    A {from=MaterialExpressionMultiply_6}
    B {from=MaterialExpressionReroute_3}
   ]
  }
  MaterialExpressionNamedRerouteDeclaration_1 {
   class=MaterialExpressionNamedRerouteDeclaration
   inputs [None {from=MaterialExpressionScalarParameter_42}]
   properties {name=TimestampLength, nodeColor=<rgba 0.0781 0 1 1>}
  }
  MaterialExpressionNamedRerouteDeclaration_45 {
   class=MaterialExpressionNamedRerouteDeclaration
   inputs [None {from=MaterialExpressionTextureSample_2, output=G}]
   properties {name=RandomID, nodeColor=<rgba 0 0.1797 1 1>}
  }
  MaterialExpressionNamedRerouteDeclaration_46 {
   class=MaterialExpressionNamedRerouteDeclaration
   inputs [None {from=MaterialExpressionTextureSample_2, output=R}]
   properties {name=CustomIDMap, nodeColor=<rgba 0.5 1 0 1>}
  }
  MaterialExpressionNamedRerouteDeclaration_47 {
   class=MaterialExpressionNamedRerouteDeclaration
   inputs [None {from=MaterialExpressionTextureSample_2, output=B}]
   properties {name=LampColorMask, nodeColor=<rgba 1 0 0.8203 1>}
  }
  MaterialExpressionNamedRerouteDeclaration_48 {
   class=MaterialExpressionNamedRerouteDeclaration
   inputs [None {from=MaterialExpressionSaturate_39}]
   properties {name=L-Mask, nodeColor=<rgba 1 0 0.8672 1>}
  }
  MaterialExpressionNamedRerouteDeclaration_49 {
   class=MaterialExpressionNamedRerouteDeclaration
   inputs [None {from=MaterialExpressionSaturate_34}]
   properties {name=W-Mask, nodeColor=<rgba 0 1 0.8125 1>}
  }
  MaterialExpressionNamedRerouteDeclaration_50 {
   class=MaterialExpressionNamedRerouteDeclaration
   inputs [None {from=MaterialExpressionSaturate_33}]
   properties {name=I-Mask, nodeColor=<rgba 1 0.4922 0 1>}
  }
  MaterialExpressionNamedRerouteDeclaration_51 {
   class=MaterialExpressionNamedRerouteDeclaration
   inputs [None {from=MaterialExpressionSaturate_38}]
   properties {name=YellowLamps, nodeColor=<rgba 1 0.9259 0 1>}
  }
  MaterialExpressionNamedRerouteDeclaration_52 {
   class=MaterialExpressionNamedRerouteDeclaration
   inputs [None {from=MaterialExpressionSaturate_35}]
   properties {name=GreenLamps, nodeColor=<rgba 0.074 1 0 1>}
  }
  MaterialExpressionNamedRerouteDeclaration_53 {
   class=MaterialExpressionNamedRerouteDeclaration
   inputs [None {from=MaterialExpressionSaturate_36}]
   properties {name=BlueLamps, nodeColor=<rgba 0 0.0338 1 1>}
  }
  MaterialExpressionNamedRerouteDeclaration_54 {
   class=MaterialExpressionNamedRerouteDeclaration
   inputs [None {from=MaterialExpressionSaturate_37}]
   properties {name=RedLamps, nodeColor=<rgba 1 0 0.0114 1>}
  }
  MaterialExpressionNamedRerouteDeclaration_55 {
   class=MaterialExpressionNamedRerouteDeclaration
   inputs [None {from=MaterialExpressionSwitch_5}]
   properties {name=LampColorIDMask, nodeColor=<rgba 1 0.3516 0 1>}
  }
  MaterialExpressionNamedRerouteDeclaration_56 {
   class=MaterialExpressionNamedRerouteDeclaration
   inputs [None {from=MaterialExpressionStaticSwitchParameter_7}]
   properties {name=LightsOn, nodeColor=<rgba 0.0312 1 0 1>}
  }
  MaterialExpressionNamedRerouteDeclaration_57 {
   class=MaterialExpressionNamedRerouteDeclaration
   inputs [None {from=MaterialExpressionMaterialFunctionCall_14, output=Result}]
   properties {name=PreReactivity_LightsOff1, nodeColor=<rgba 0 1 0.3906 1>}
  }
  MaterialExpressionNamedRerouteDeclaration_58 {
   class=MaterialExpressionNamedRerouteDeclaration
   inputs [None {from=MaterialExpressionMaterialFunctionCall_15, output=Result}]
   properties {name=PreReactivity_LightsFlicker, nodeColor=<rgba 1 0.0703 0 1>}
  }
  MaterialExpressionNamedRerouteDeclaration_59 {
   class=MaterialExpressionNamedRerouteDeclaration
   inputs [None {from=MaterialExpressionMaterialFunctionCall_16, output=Result}]
   properties {name=Reactivity, nodeColor=<rgba 1 0 0.0291 1>}
  }
  MaterialExpressionNamedRerouteDeclaration_60 {
   class=MaterialExpressionNamedRerouteDeclaration
   inputs [None {from=MaterialExpressionCeil_0}]
   properties {name=YellowLightReactivityLoop, nodeColor=<rgba 0 1 0.9062 1>}
  }
  MaterialExpressionNamedRerouteDeclaration_61 {
   class=MaterialExpressionNamedRerouteDeclaration
   inputs [None {from=MaterialExpressionMultiply_45}]
   properties {name=YellowLightBlink, nodeColor=<rgba 1 0.5859 0 1>}
  }
  MaterialExpressionNamedRerouteDeclaration_62 {
   class=MaterialExpressionNamedRerouteDeclaration
   inputs [None {from=MaterialExpressionCeil_2}]
   properties {name=GameplayReactivity, nodeColor=<rgba 0.7109 0 1 1>}
  }
  MaterialExpressionNamedRerouteDeclaration_63 {
   class=MaterialExpressionNamedRerouteDeclaration
   inputs [None {from=MaterialExpressionMaterialFunctionCall_17, output=Result}]
   properties {name=PreReactivity_LightsOff2, nodeColor=<rgba 0.5 1 0 1>}
  }
  MaterialExpressionNamedRerouteUsage_0 {class=MaterialExpressionNamedRerouteUsage}
  MaterialExpressionNamedRerouteUsage_10 {class=MaterialExpressionNamedRerouteUsage}
  MaterialExpressionNamedRerouteUsage_58 {class=MaterialExpressionNamedRerouteUsage}
  MaterialExpressionNamedRerouteUsage_59 {class=MaterialExpressionNamedRerouteUsage}
  MaterialExpressionNamedRerouteUsage_60 {class=MaterialExpressionNamedRerouteUsage}
  MaterialExpressionNamedRerouteUsage_61 {class=MaterialExpressionNamedRerouteUsage}
  MaterialExpressionNamedRerouteUsage_62 {class=MaterialExpressionNamedRerouteUsage}
  MaterialExpressionNamedRerouteUsage_63 {class=MaterialExpressionNamedRerouteUsage}
  MaterialExpressionNamedRerouteUsage_64 {class=MaterialExpressionNamedRerouteUsage}
  MaterialExpressionNamedRerouteUsage_65 {class=MaterialExpressionNamedRerouteUsage}
  MaterialExpressionNamedRerouteUsage_66 {class=MaterialExpressionNamedRerouteUsage}
  MaterialExpressionNamedRerouteUsage_67 {class=MaterialExpressionNamedRerouteUsage}
  MaterialExpressionNamedRerouteUsage_68 {class=MaterialExpressionNamedRerouteUsage}
  MaterialExpressionNamedRerouteUsage_69 {class=MaterialExpressionNamedRerouteUsage}
  MaterialExpressionNamedRerouteUsage_70 {class=MaterialExpressionNamedRerouteUsage}
  MaterialExpressionNamedRerouteUsage_71 {class=MaterialExpressionNamedRerouteUsage}
  MaterialExpressionNamedRerouteUsage_72 {class=MaterialExpressionNamedRerouteUsage}
  MaterialExpressionNamedRerouteUsage_73 {class=MaterialExpressionNamedRerouteUsage}
  MaterialExpressionNamedRerouteUsage_74 {class=MaterialExpressionNamedRerouteUsage}
  MaterialExpressionNamedRerouteUsage_75 {class=MaterialExpressionNamedRerouteUsage}
  MaterialExpressionNamedRerouteUsage_76 {class=MaterialExpressionNamedRerouteUsage}
  MaterialExpressionNamedRerouteUsage_77 {class=MaterialExpressionNamedRerouteUsage}
  MaterialExpressionNamedRerouteUsage_78 {class=MaterialExpressionNamedRerouteUsage}
  MaterialExpressionNamedRerouteUsage_79 {class=MaterialExpressionNamedRerouteUsage}
  MaterialExpressionNamedRerouteUsage_80 {class=MaterialExpressionNamedRerouteUsage}
  MaterialExpressionNamedRerouteUsage_81 {class=MaterialExpressionNamedRerouteUsage}
  MaterialExpressionOneMinus_4 {class=MaterialExpressionOneMinus, inputs [None {from=MaterialExpressionSaturate_19}]}
  MaterialExpressionOneMinus_6 {class=MaterialExpressionOneMinus, inputs [None {from=MaterialExpressionSaturate_46}]}
  MaterialExpressionOneMinus_8 {class=MaterialExpressionOneMinus, inputs [None {from=MaterialExpressionIf_2}]}
  MaterialExpressionParticleColor_0 {class=MaterialExpressionParticleColor, inputs [], properties {}}
  MaterialExpressionParticleDirection_0 {class=MaterialExpressionParticleDirection, inputs [], properties {}}
  MaterialExpressionParticleSize_1 {class=MaterialExpressionParticleSize, inputs [], properties {}}
  MaterialExpressionParticleSpeed_0 {class=MaterialExpressionParticleSpeed, inputs [], properties {}}
  MaterialExpressionPower_0 {
   class=MaterialExpressionPower
   inputs [
    Base {from=MaterialExpressionStaticSwitchParameter_2}
    Exp {from=MaterialExpressionDynamicParameter_0, output=Power}
   ]
  }
  MaterialExpressionPreviousFrameSwitch_1 {
   class=MaterialExpressionPreviousFrameSwitch
   inputs [
    Current Frame {from=MaterialExpressionReroute_1, output=""}
    Previous Frame {from=MaterialExpressionSubtract_0, output=""}
   ]
   properties {}
  }
  MaterialExpressionReroute_0 {class=MaterialExpressionReroute, inputs [None {from=MaterialExpressionPower_0}]}
  MaterialExpressionReroute_1 {class=MaterialExpressionReroute, inputs [None {from=MaterialExpressionStaticSwitchParameter_1}]}
  MaterialExpressionReroute_2 {class=MaterialExpressionReroute, inputs [None {from=MaterialExpressionDynamicParameter_0, output=HDRMult}]}
  MaterialExpressionReroute_23 {class=MaterialExpressionReroute, inputs [None {from=MaterialExpressionLinearInterpolate_19}]}
  MaterialExpressionReroute_24 {class=MaterialExpressionReroute, inputs [None {from=MaterialExpressionLinearInterpolate_19}]}
  MaterialExpressionReroute_25 {class=MaterialExpressionReroute, inputs [None {from=MaterialExpressionMultiply_43}]}
  MaterialExpressionReroute_26 {class=MaterialExpressionReroute, inputs [None {from=MaterialExpressionFrac_9}]}
  MaterialExpressionReroute_27 {class=MaterialExpressionReroute, inputs [None {from=MaterialExpressionLinearInterpolate_19}]}
  MaterialExpressionReroute_28 {class=MaterialExpressionReroute, inputs [None {from=MaterialExpressionConstantBiasScale_1}]}
  MaterialExpressionReroute_29 {class=MaterialExpressionReroute, inputs [None {from=MaterialExpressionLinearInterpolate_19}]}
  MaterialExpressionReroute_3 {class=MaterialExpressionReroute, inputs [None {from=MaterialExpressionNamedRerouteDeclaration_56}]}
  MaterialExpressionReroute_30 {class=MaterialExpressionReroute, inputs [None {from=MaterialExpressionMultiply_46}]}
  MaterialExpressionReroute_31 {class=MaterialExpressionReroute, inputs [None {from=MaterialExpressionSmoothStep_1}]}
  MaterialExpressionSaturate_18 {class=MaterialExpressionSaturate, inputs [None {from=MaterialExpressionDivide_7}]}
  MaterialExpressionSaturate_19 {class=MaterialExpressionSaturate, inputs [None {from=MaterialExpressionDivide_8}]}
  MaterialExpressionSaturate_33 {class=MaterialExpressionSaturate, inputs [None {from=MaterialExpressionSubtract_16}]}
  MaterialExpressionSaturate_34 {class=MaterialExpressionSaturate, inputs [None {from=MaterialExpressionSubtract_15}]}
  MaterialExpressionSaturate_35 {class=MaterialExpressionSaturate, inputs [None {from=MaterialExpressionSubtract_17}]}
  MaterialExpressionSaturate_36 {class=MaterialExpressionSaturate, inputs [None {from=MaterialExpressionSubtract_18}]}
  MaterialExpressionSaturate_37 {class=MaterialExpressionSaturate, inputs [None {from=MaterialExpressionStep_37}]}
  MaterialExpressionSaturate_38 {class=MaterialExpressionSaturate, inputs [None {from=MaterialExpressionStep_34}]}
  MaterialExpressionSaturate_39 {class=MaterialExpressionSaturate, inputs [None {from=MaterialExpressionStep_31}]}
  MaterialExpressionSaturate_40 {class=MaterialExpressionSaturate, inputs [None {from=MaterialExpressionAdd_10}]}
  MaterialExpressionSaturate_41 {class=MaterialExpressionSaturate, inputs [None {from=MaterialExpressionSine_11}]}
  MaterialExpressionSaturate_42 {class=MaterialExpressionSaturate, inputs [None {from=MaterialExpressionCosine_2}]}
  MaterialExpressionSaturate_43 {class=MaterialExpressionSaturate, inputs [None {from=MaterialExpressionSine_12}]}
  MaterialExpressionSaturate_44 {class=MaterialExpressionSaturate, inputs [None {from=MaterialExpressionSwitch_6}]}
  MaterialExpressionSaturate_45 {class=MaterialExpressionSaturate, inputs [None {from=MaterialExpressionDivide_13}]}
  MaterialExpressionSaturate_46 {class=MaterialExpressionSaturate, inputs [None {from=MaterialExpressionDivide_14}]}
  MaterialExpressionSaturate_47 {class=MaterialExpressionSaturate, inputs [None {from=MaterialExpressionSine_13}]}
  MaterialExpressionSaturate_5 {class=MaterialExpressionSaturate, inputs [None {from=MaterialExpressionSine_3}]}
  MaterialExpressionScalarParameter_0 {class=MaterialExpressionScalarParameter, properties {defaultValue=0.75, parameterName=Cap Size Multiplier, sliderMax=1}}
  MaterialExpressionScalarParameter_1 {class=MaterialExpressionScalarParameter, properties {defaultValue=1, parameterName=Glow Bright Value}}
  MaterialExpressionScalarParameter_2 {class=MaterialExpressionScalarParameter, properties {parameterName=Glow Dark Value}}
  MaterialExpressionScalarParameter_3 {class=MaterialExpressionScalarParameter, properties {defaultValue=0.1, parameterName=TeardropMinWidth}}
  MaterialExpressionScalarParameter_35 {class=MaterialExpressionScalarParameter, properties {parameterName=ReactivityID}}
  MaterialExpressionScalarParameter_36 {class=MaterialExpressionScalarParameter, properties {defaultValue=0.2, parameterName=LightsSpeed}}
  MaterialExpressionScalarParameter_37 {class=MaterialExpressionScalarParameter, properties {defaultValue=4, parameterName=YellowLightsReactivitySpeed}}
  MaterialExpressionScalarParameter_38 {class=MaterialExpressionScalarParameter, properties {defaultValue=-999, group=MagmaBreak, parameterName=ReactivityTimestamp}}
  MaterialExpressionScalarParameter_39 {class=MaterialExpressionScalarParameter, properties {defaultValue=5, parameterName=LightsSpeed_PreReactivity}}
  MaterialExpressionScalarParameter_4 {class=MaterialExpressionScalarParameter, properties {parameterName=SphereMaskHardness}}
  MaterialExpressionScalarParameter_40 {class=MaterialExpressionScalarParameter, properties {defaultValue=0.4, parameterName=LampsFlickerFrame}}
  MaterialExpressionScalarParameter_41 {class=MaterialExpressionScalarParameter, properties {defaultValue=0.2, parameterName=LampsOff1Frame}}
  MaterialExpressionScalarParameter_42 {class=MaterialExpressionScalarParameter, properties {defaultValue=5, parameterName=ReactivityTimestampLength}}
  MaterialExpressionScalarParameter_43 {class=MaterialExpressionScalarParameter, properties {defaultValue=0.6, parameterName=LampsOff2Frame}}
  MaterialExpressionScalarParameter_44 {class=MaterialExpressionScalarParameter, properties {defaultValue=0.5, parameterName=ReactivityBlinkSpeed}}
  MaterialExpressionScalarParameter_45 {class=MaterialExpressionScalarParameter, properties {defaultValue=-999, group=MagmaBreak, parameterName=YellowLightsTimestamp}}
  MaterialExpressionScalarParameter_46 {class=MaterialExpressionScalarParameter, properties {defaultValue=7, parameterName=YellowLightsTimestampLength}}
  MaterialExpressionScalarParameter_47 {class=MaterialExpressionScalarParameter, properties {defaultValue=0.91, parameterName=IdleBlinkThreshold}}
  MaterialExpressionScalarParameter_5 {class=MaterialExpressionScalarParameter, properties {parameterName=Camera depth Fade length}}
  MaterialExpressionScalarParameter_6 {class=MaterialExpressionScalarParameter, properties {parameterName=Camera depth Fade Offset}}
  MaterialExpressionShadingPathSwitch_0 {
   class=MaterialExpressionShadingPathSwitch
   inputs [
    Default {from=MaterialExpressionDepthFade_0, output=""}
    Deferred {from="", output=""}
    Forward {from=MaterialExpressionReroute_0, output=""}
    Mobile {from=MaterialExpressionReroute_0, output=""}
   ]
   properties {}
  }
  MaterialExpressionShadingPathSwitch_1 {
   class=MaterialExpressionShadingPathSwitch
   inputs [
    Default {from=MaterialExpressionMaterialFunctionCall_0, output=RemappedRange}
    Deferred {from="", output=""}
    Forward {from="", output=""}
    Mobile {from=MaterialExpressionMultiply_0, output=""}
   ]
   properties {}
  }
  MaterialExpressionSine_11 {class=MaterialExpressionSine, inputs [None {from=MaterialExpressionAdd_9}], properties {period=1.7435}}
  MaterialExpressionSine_12 {class=MaterialExpressionSine, inputs [None {from=MaterialExpressionMultiply_50}]}
  MaterialExpressionSine_13 {class=MaterialExpressionSine, inputs [None {from=MaterialExpressionDivide_15}]}
  MaterialExpressionSine_14 {class=MaterialExpressionSine, inputs [None {from=MaterialExpressionMultiply_55}]}
  MaterialExpressionSine_15 {class=MaterialExpressionSine, inputs [None {from=MaterialExpressionMultiply_53}]}
  MaterialExpressionSine_3 {class=MaterialExpressionSine, inputs [None {from=MaterialExpressionMultiply_56}]}
  MaterialExpressionSmoothStep_1 {
   class=MaterialExpressionSmoothStep
   inputs [
    Min {from=MaterialExpressionScalarParameter_47, output=""}
    Max {from="", output=""}
    Value {from=MaterialExpressionSaturate_40, output=""}
   ]
   properties {constMin=0.8}
  }
  MaterialExpressionSphereMask_0 {
   class=MaterialExpressionSphereMask
   inputs [
    A {from=MaterialExpressionStaticSwitchParameter_0, output=""}
    B {from=MaterialExpressionConstant_1, output=""}
    Radius {from=MaterialExpressionConstant_0, output=""}
    Hardness {from=MaterialExpressionScalarParameter_4, output=""}
   ]
   properties {hardnessPercent=0}
  }
  MaterialExpressionStaticBool_0 {class=MaterialExpressionStaticBool}
  MaterialExpressionStaticSwitchParameter_0 {
   class=MaterialExpressionStaticSwitchParameter
   inputs [
    True {from=MaterialExpressionMaterialFunctionCall_2, output=Texture Coordinates}
    False {from=MaterialExpressionMaterialFunctionCall_4, output=Texture Coordinates}
   ]
   properties {parameterName=UseTeardropShape}
  }
  MaterialExpressionStaticSwitchParameter_1 {
   class=MaterialExpressionStaticSwitchParameter
   inputs [
    True {from=MaterialExpressionMaterialFunctionCall_2, output=World Position Offset}
    False {from=MaterialExpressionMaterialFunctionCall_4, output=World Position Offset}
   ]
   properties {parameterName=UseTeardropShape}
  }
  MaterialExpressionStaticSwitchParameter_2 {
   class=MaterialExpressionStaticSwitchParameter
   inputs [
    True {from=MaterialExpressionMultiply_5, output=""}
    False {from=MaterialExpressionSphereMask_0, output=""}
   ]
   properties {parameterName=UseEndFade}
  }
  MaterialExpressionStaticSwitchParameter_7 {
   class=MaterialExpressionStaticSwitchParameter
   inputs [
    True {from=MaterialExpressionLinearInterpolate_22, output=""}
    False {from=MaterialExpressionReroute_31, output=""}
   ]
   properties {parameterName=EnableReactivity}
  }
  MaterialExpressionStaticSwitchParameter_8 {
   class=MaterialExpressionStaticSwitchParameter
   inputs [
    True {from=MaterialExpressionLinearInterpolate_17, output=""}
    False {from=MaterialExpressionScalarParameter_36, output=""}
   ]
   properties {parameterName=EnableReactivity}
  }
  MaterialExpressionStep_31 {
   class=MaterialExpressionStep
   inputs [
    Y {from="", output=""}
    X {from=MaterialExpressionNamedRerouteUsage_58, output=""}
   ]
   properties {constX=0.3, constY=1}
  }
  MaterialExpressionStep_32 {
   class=MaterialExpressionStep
   inputs [
    Y {from="", output=""}
    X {from=MaterialExpressionNamedRerouteUsage_58, output=""}
   ]
   properties {constX=0.64, constY=0.64}
  }
  MaterialExpressionStep_33 {
   class=MaterialExpressionStep
   inputs [
    Y {from="", output=""}
    X {from=MaterialExpressionNamedRerouteUsage_58, output=""}
   ]
   properties {constX=0.64, constY=0.33}
  }
  MaterialExpressionStep_34 {
   class=MaterialExpressionStep
   inputs [
    Y {from=MaterialExpressionNamedRerouteUsage_62, output=""}
    X {from="", output=""}
   ]
   properties {constX=0.25, constY=0.3}
  }
  MaterialExpressionStep_35 {
   class=MaterialExpressionStep
   inputs [
    Y {from=MaterialExpressionNamedRerouteUsage_62, output=""}
    X {from="", output=""}
   ]
   properties {constX=0.6, constY=0.3}
  }
  MaterialExpressionStep_36 {
   class=MaterialExpressionStep
   inputs [
    Y {from=MaterialExpressionNamedRerouteUsage_62, output=""}
    X {from="", output=""}
   ]
   properties {constX=0.8, constY=0.3}
  }
  MaterialExpressionStep_37 {
   class=MaterialExpressionStep
   inputs [
    Y {from="", output=""}
    X {from=MaterialExpressionNamedRerouteUsage_62, output=""}
   ]
   properties {constY=1}
  }
  MaterialExpressionStep_38 {
   class=MaterialExpressionStep
   inputs [
    Y {from="", output=""}
    X {from=MaterialExpressionNamedRerouteUsage_77, output=""}
   ]
   properties {constX=0.01, constY=0.01}
  }
  MaterialExpressionStep_39 {
   class=MaterialExpressionStep
   inputs [
    Y {from="", output=""}
    X {from=MaterialExpressionNamedRerouteUsage_80, output=""}
   ]
   properties {constX=0.01, constY=0.01}
  }
  MaterialExpressionStep_40 {
   class=MaterialExpressionStep
   inputs [
    Y {from="", output=""}
    X {from=MaterialExpressionSine_15, output=""}
   ]
   properties {constX=0.01, constY=0.01}
  }
  MaterialExpressionStep_41 {
   class=MaterialExpressionStep
   inputs [
    Y {from="", output=""}
    X {from=MaterialExpressionComponentMask_4, output=""}
   ]
   properties {constX=0.5, constY=0.5}
  }
  MaterialExpressionStep_42 {
   class=MaterialExpressionStep
   inputs [
    Y {from="", output=""}
    X {from=MaterialExpressionNamedRerouteUsage_78, output=""}
   ]
   properties {constX=0.01, constY=0.01}
  }
  MaterialExpressionStep_43 {
   class=MaterialExpressionStep
   inputs [
    Y {from="", output=""}
    X {from=MaterialExpressionNamedRerouteUsage_79, output=""}
   ]
   properties {constX=0.01, constY=0.01}
  }
  MaterialExpressionStep_44 {
   class=MaterialExpressionStep
   inputs [
    Y {from="", output=""}
    X {from=MaterialExpressionNamedRerouteUsage_74, output=""}
   ]
   properties {constX=0.01, constY=0.01}
  }
  MaterialExpressionStep_45 {
   class=MaterialExpressionStep
   inputs [
    Y {from="", output=""}
    X {from=MaterialExpressionSaturate_47, output=""}
   ]
   properties {constY=0.01}
  }
  MaterialExpressionStep_8 {
   class=MaterialExpressionStep
   inputs [
    Y {from="", output=""}
    X {from=MaterialExpressionSaturate_5, output=""}
   ]
   properties {constY=0.01}
  }
  MaterialExpressionSubtract_0 {
   class=MaterialExpressionSubtract
   inputs [
    A {from=MaterialExpressionReroute_1}
    B {from=MaterialExpressionMultiply_4}
   ]
  }
  MaterialExpressionSubtract_15 {
   class=MaterialExpressionSubtract
   inputs [
    A {from=MaterialExpressionStep_32}
    B {from=MaterialExpressionNamedRerouteUsage_59}
   ]
  }
  MaterialExpressionSubtract_16 {
   class=MaterialExpressionSubtract
   inputs [
    A {from=MaterialExpressionStep_33}
    B {from=MaterialExpressionMax_5}
   ]
  }
  MaterialExpressionSubtract_17 {
   class=MaterialExpressionSubtract
   inputs [
    A {from=MaterialExpressionStep_35}
    B {from=MaterialExpressionNamedRerouteUsage_63}
   ]
  }
  MaterialExpressionSubtract_18 {
   class=MaterialExpressionSubtract
   inputs [
    A {from=MaterialExpressionStep_36}
    B {from=MaterialExpressionMax_6}
   ]
  }
  MaterialExpressionSubtract_19 {
   class=MaterialExpressionSubtract
   inputs [
    A {from=MaterialExpressionTime_10}
    B {from=MaterialExpressionDynamicParameter_6, output=ReactivityTimestamp}
   ]
  }
  MaterialExpressionSubtract_20 {
   class=MaterialExpressionSubtract
   inputs [
    A {from=MaterialExpressionTime_10}
    B {from=MaterialExpressionAdd_11}
   ]
  }
  MaterialExpressionSubtract_8 {
   class=MaterialExpressionSubtract
   inputs [
    A {from=MaterialExpressionTime_3}
    B {from=MaterialExpressionDynamicParameter_7, output=YellowLightsTimestamp}
   ]
  }
  MaterialExpressionSubtract_9 {
   class=MaterialExpressionSubtract
   inputs [
    A {from=MaterialExpressionTime_3}
    B {from=MaterialExpressionAdd_4}
   ]
  }
  MaterialExpressionSwitch_5 {
   class=MaterialExpressionSwitch
   inputs [
    SwitchValue {from=MaterialExpressionDynamicParameter_5, output=ReactivityID}
    Default {from="", output=""}
    None {from=MaterialExpressionNamedRerouteUsage_69, output=""}
    None {from=MaterialExpressionNamedRerouteUsage_68, output=""}
    None {from=MaterialExpressionNamedRerouteUsage_67, output=""}
    None {from=MaterialExpressionNamedRerouteUsage_66, output=""}
   ]
   properties {inputs [
     {
      inputName=None
      input {
       expression=&self:MaterialExpressionNamedRerouteUsage_69
       inputName=None
       mask=0
       maskA=0
       maskB=0
       maskG=0
       maskR=0
       outputIndex=0
      }
     }
     {
      inputName=None
      input {
       expression=&self:MaterialExpressionNamedRerouteUsage_68
       inputName=None
       mask=0
       maskA=0
       maskB=0
       maskG=0
       maskR=0
       outputIndex=0
      }
     }
     {
      inputName=None
      input {
       expression=&self:MaterialExpressionNamedRerouteUsage_67
       inputName=None
       mask=0
       maskA=0
       maskB=0
       maskG=0
       maskR=0
       outputIndex=0
      }
     }
     {
      inputName=None
      input {
       expression=&self:MaterialExpressionNamedRerouteUsage_66
       inputName=None
       mask=0
       maskA=0
       maskB=0
       maskG=0
       maskR=0
       outputIndex=0
      }
     }
    ]}
  }
  MaterialExpressionSwitch_6 {
   class=MaterialExpressionSwitch
   inputs [
    SwitchValue {from=MaterialExpressionMultiply_44, output=""}
    Default {from="", output=""}
    None {from=MaterialExpressionNamedRerouteUsage_70, output=""}
    None {from=MaterialExpressionNamedRerouteUsage_71, output=""}
    None {from=MaterialExpressionNamedRerouteUsage_72, output=""}
    None {from=MaterialExpressionNamedRerouteUsage_72, output=""}
   ]
   properties {
    constDefault=1
    inputs [
     {
      inputName=None
      input {
       expression=&self:MaterialExpressionNamedRerouteUsage_70
       inputName=None
       mask=0
       maskA=0
       maskB=0
       maskG=0
       maskR=0
       outputIndex=0
      }
     }
     {
      inputName=None
      input {
       expression=&self:MaterialExpressionNamedRerouteUsage_71
       inputName=None
       mask=0
       maskA=0
       maskB=0
       maskG=0
       maskR=0
       outputIndex=0
      }
     }
     {
      inputName=None
      input {
       expression=&self:MaterialExpressionNamedRerouteUsage_72
       inputName=None
       mask=0
       maskA=0
       maskB=0
       maskG=0
       maskR=0
       outputIndex=0
      }
     }
     {
      inputName=None
      input {
       expression=&self:MaterialExpressionNamedRerouteUsage_72
       inputName=None
       mask=0
       maskA=0
       maskB=0
       maskG=0
       maskR=0
       outputIndex=0
      }
     }
    ]
   }
  }
  MaterialExpressionTextureCoordinate_0 {class=MaterialExpressionTextureCoordinate}
  MaterialExpressionTextureCoordinate_4 {class=MaterialExpressionTextureCoordinate}
  MaterialExpressionTextureCoordinate_5 {class=MaterialExpressionTextureCoordinate, properties {uTiling=0.01, vTiling=0.01}}
  MaterialExpressionTextureSample_2 {
   class=MaterialExpressionTextureSample
   inputs [
    UVs {from=MaterialExpressionAdd_14}
    Tex {from=""}
    Apply View MipBias {from=""}
   ]
   properties {
    constCoordinate=1
    samplerType=SAMPLERTYPE_LinearColor
    texture=/BRCosmetics/Accessories/FORT_Backpacks/Backpack_M_MED_LanternFit/Textures/T_LanternFit_IDs
   }
  }
  MaterialExpressionTime_10 {class=MaterialExpressionTime}
  MaterialExpressionTime_11 {class=MaterialExpressionTime}
  MaterialExpressionTime_12 {class=MaterialExpressionTime}
  MaterialExpressionTime_3 {class=MaterialExpressionTime}
  MaterialExpressionTime_9 {class=MaterialExpressionTime}
 }
 properties {
  bEnableResponsiveAA=true
  bUsedWithNiagaraSprites=true
  blendMode=BLEND_Additive
  pixelDepthOffsetMode=PDOM_Legacy
  shadingModel=MSM_Unlit
  usedShadingModels=Unlit
 }
}
