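# Idle Niagara effect for the LanternFit pickaxe that places twelve persistent, colored capsule glow sprites on the weapon's sockets.
# A single burst gives each particle a spawn group index, which selects its socket, color and pair of dynamic material values from arrays.
# Sprites follow their sockets, aligning to socket rotation with a slight camera offset.
# Glow alpha scales with a time-of-day float and a front-end versus gameplay modifier.
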NiagaraSystem /BRCosmetics/Weapons/FORT_Melee/Pickaxe_LanternFit/FX/NS_Pickaxe_LanternFit_Idle
dependencies {
 usedDataInterfaces [
  &NiagaraDataInterfaceArrayColor
  &NiagaraDataInterfaceArrayFloat
  &NiagaraDataInterfaceCurve
  &NiagaraDataInterfaceSocketReader
  &NiagaraDataInterfaceColorCurve
 ]
 usedDynamicInputs [&ColorFromCurve, &FNTimeOfDayScalar]
 usedModules [
  &self:SystemSpawnScript.SystemScriptSource.SystemScriptGraph.NiagaraNodeAssignment_2.SetVariables_#1
  &SystemState
  &self:SystemSpawnScript.SystemScriptSource.SystemScriptGraph.NiagaraNodeAssignment_1.SetVariables_#2
 ]
}
emitters {CapsuleGlow {
  emitterUpdateScript [
   &EmitterState(
    Inactive Response=::NewEnumerator1
    Life Cycle Mode=::NewEnumerator1
    Loop Behavior=::NewEnumerator1
    Loop Duration Mode=::NewEnumerator1
    UseLoopDelay=false
   )
   &SpawnBurst_Instantaneous(Spawn Count=12, Use Spawn Probability=false, UseLoopCountLimit=false)
  ]
  particleSpawnScript [
   &FN_SetSpawnGroup(Number of Spawn Groups=12)
   &InitializeParticle(
    Color=&SelectLinearColorFromArray(
     Array Sampling Mode=::NewEnumerator1
     Direct Array Index=&SpawnGroup()
     Color Selection Array=DataInterface() {propertyValues {ColorData [
        <rgba 1 0.0003 0.0215 1>
        <rgba 0.1544 0.82 0.0445 1>
        <rgba 0.7 0.5863 0 1>
        <rgba 1 0.0003 0.0215 1>
        <rgba 0.0393 0.4598 1.2 1>
        <rgba 0.1544 0.82 0.0445 1>
        <rgba 0.7 0.5863 0 1>
        <rgba 1 0.0003 0.0215 1>
        <rgba 0.0393 0.4598 1.2 1>
        <rgba 0.1544 0.82 0.0445 1>
        <rgba 0.7 0.5863 0 1>
        <rgba 0.0393 0.4598 1.2 1>
       ]}}
     Random Seed=?
    )
    Sprite Rotation Mode=::NewEnumerator0
    Sprite Size=<17.421 25.3361>
    Sprite Size Mode=::NewEnumerator4
    UsePositionOffset=true
    Write Material Random=true
   )
   &CameraOffset(Camera Offset Amount=3)
   &DynamicMaterialParameters(
    Index 2 Param 1=&SelectFloatFromArray(
     Direct Array Index=&SpawnGroup()
     Float Selection Array=DataInterface() {propertyValues {FloatData [0.75,0.25,0,0.75,0.5,0.25,0,0.75,0.5,0.25,0,0.5]}}
    )
    Index 2 Param 2=&SelectFloatFromArray(
     Direct Array Index=&SpawnGroup()
     Float Selection Array=DataInterface() {propertyValues {FloatData [0.5,0.5,0.5,0.25,0.25,0.25,0.25,0,0,0,0,0.5]}}
    )
    Module.Index2Float0Write=true
    Module.Index2Float1Write=true
    Module.Index2Float2Write=false
    Module.Index2Float3Write=false
    Write Parameter Index 0=::NewEnumerator0
    Write Parameter Index 2=::NewEnumerator1
   )
  ]
  particleUpdateScript [
   &ParticleState(Kill Particles When Lifetime Has Elapsed=false, Let Infinitely Lived Particles Die When Emitter Deactivates=true)
   &ScaleSpriteSize(Scale Sprite Size Mode=::NewEnumerator0, Uniform Scale Factor=0.9038)
   &ScaleColor(
    Scale Alpha=&Multiply_Float(
     A=&Multiply_Float(A=&FN_FrontEndFloatModifier(Gameplay=0.1, Front End=0.05), B=$System.TODFloat)
     B=0.508
    )
    Scale RGB=&MakeVectorFromLinearColorRGB(Color=<rgba 2 2 2 1>)
    ScaleA=true
    ScaleRGB=true
   )
   &SocketLocation(
    SamplePosOffset=false
    Socket Index=&SpawnGroup()
    Socket Reader=$System.Socket Reader
    Socket Sampling Mode=::NewEnumerator2
    VelocitySpeedLimit=false
   )
   &SpriteFacingAndAlignment(
    Sprite Alignment=&RotateVectorByQuaternion(Vector To Rotate=<1 0 0>, Quaternion=$Output.SocketLocation.SampledRotation)
    WriteAlignment=true
    WriteFacing=true
   ) {module=SpriteFacingAndAlignment001}
  ]
  rendererClasses [&NiagaraSpriteRendererProperties]
  renderers [{rendererClass=&NiagaraSpriteRendererProperties}]
 }}
summary {
 systemName=NS_Pickaxe_LanternFit_Idle
 emitters [{emitterName=CapsuleGlow, rendererClasses [&NiagaraSpriteRendererProperties]}]
 userVariables [
  User.bIsFrontEnd {type=NiagaraBool}
  User.bIsFrontEndPreview {type=NiagaraBool}
 ]
}
