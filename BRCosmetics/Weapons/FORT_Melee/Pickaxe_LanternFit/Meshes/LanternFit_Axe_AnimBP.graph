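# Animation Blueprint for the LanternFit pickaxe that adds physics sway to its dangling lightbulbs and wires.
# The input pose passes through a chain of spring-based Anim Dynamics nodes: twelve lightbulb bones (no bulb 2), then three wire bones.
# Bulb simulation-space pins are set to World, wire pins to Component.
# The update event sets SimlationSpace to World in the front end, otherwise Component, but the AnimGraph never reads it.
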AnimBlueprint /BRCosmetics/Weapons/FORT_Melee/Pickaxe_LanternFit/Meshes/LanternFit_Axe_AnimBP
target_skeleton=&/LanternFit_Axe_Skeleton
graphs [
 AnimGraph {nodes [
   AnimGraphNode_SubInput_0 = Input Pose()
   n1 = Anim Dynamics - Bone: dyn_lightbulb_12!(
    SimulationSpace:AnimPhysSimSpaceType Enum=World
    ComponentPose=Local To Component(LocalPose=@AnimGraphNode_SubInput_0)
    Alpha=1
    ExternalForce="0,0,0"
   ) -> (Pose) {
    angularDampingOverride=0.1
    angularSpringConstant=500
    bAngularSpring=true
    bOverrideLinearDamping=true
    bUseGravityOverride=true
    linearDampingOverride=1
    simulationSpace=World
    alphaBoolBlend {blendOption=Linear}
    boundBone {boneName=dyn_lightbulb_12}
    componentAppliedLinearAccClamp=<1000 1000 1000>
    componentLinearAccScale=<1 1 1>
    componentLinearVelScale=<1 1 1>
    gravityOverride {z=-980}
    physicsBodyDefinitions [{
      boundBone {boneName=dyn_lightbulb_12}
      boxExtents=<22 22 22>
      constraintSetup {twistAxis=AxisZ, angularLimitsMax=<1 15 15>, angularLimitsMin=<-1 -15 -15>, angularTarget=<0 0 0>}
      localJointOffset=<0 0 0.1>
     }]
    simSpaceSettings {maxAngularAcceleration=150, maxAngularVelocity=150, simSpaceAngularAlpha=1}
   }
   n2 = Anim Dynamics - Bone: dyn_lightbulb_11!(SimulationSpace:AnimPhysSimSpaceType Enum=World, ComponentPose=@n1, Alpha=1, ExternalForce="0,0,0") -> (Pose) {
    angularDampingOverride=0.1
    angularSpringConstant=500
    bAngularSpring=true
    bOverrideLinearDamping=true
    bUseGravityOverride=true
    linearDampingOverride=1
    simulationSpace=World
    alphaBoolBlend {blendOption=Linear}
    boundBone {boneName=dyn_lightbulb_11}
    componentAppliedLinearAccClamp=<1000 1000 1000>
    componentLinearAccScale=<1 1 1>
    componentLinearVelScale=<1 1 1>
    gravityOverride {z=-980}
    physicsBodyDefinitions [{
      boundBone {boneName=dyn_lightbulb_11}
      boxExtents=<22 22 22>
      constraintSetup {twistAxis=AxisZ, angularLimitsMax=<1 15 15>, angularLimitsMin=<-1 -15 -15>, angularTarget=<0 0 0>}
      localJointOffset=<0 0 0.1>
     }]
    simSpaceSettings {maxAngularAcceleration=150, maxAngularVelocity=150, simSpaceAngularAlpha=1}
   }
   n3 = Anim Dynamics - Bone: dyn_lightbulb_10!(SimulationSpace:AnimPhysSimSpaceType Enum=World, ComponentPose=@n2, Alpha=1, ExternalForce="0,0,0") -> (Pose) {
    angularDampingOverride=0.1
    angularSpringConstant=500
    bAngularSpring=true
    bOverrideLinearDamping=true
    bUseGravityOverride=true
    linearDampingOverride=1
    simulationSpace=World
    alphaBoolBlend {blendOption=Linear}
    boundBone {boneName=dyn_lightbulb_10}
    componentAppliedLinearAccClamp=<1000 1000 1000>
    componentLinearAccScale=<1 1 1>
    componentLinearVelScale=<1 1 1>
    gravityOverride {z=-980}
    physicsBodyDefinitions [{
      boundBone {boneName=dyn_lightbulb_10}
      boxExtents=<22 22 22>
      constraintSetup {twistAxis=AxisZ, angularLimitsMax=<1 15 15>, angularLimitsMin=<-1 -15 -15>, angularTarget=<0 0 0>}
      localJointOffset=<0 0 0.1>
     }]
    simSpaceSettings {maxAngularAcceleration=150, maxAngularVelocity=150, simSpaceAngularAlpha=1}
   }
   n4 = Anim Dynamics - Bone: dyn_lightbulb_9!(SimulationSpace:AnimPhysSimSpaceType Enum=World, ComponentPose=@n3, Alpha=1, ExternalForce="0,0,0") -> (Pose) {
    angularDampingOverride=0.1
    angularSpringConstant=500
    bAngularSpring=true
    bOverrideLinearDamping=true
    bUseGravityOverride=true
    linearDampingOverride=1
    simulationSpace=World
    alphaBoolBlend {blendOption=Linear}
    boundBone {boneName=dyn_lightbulb_9}
    componentAppliedLinearAccClamp=<1000 1000 1000>
    componentLinearAccScale=<1 1 1>
    componentLinearVelScale=<1 1 1>
    gravityOverride {z=-980}
    physicsBodyDefinitions [{
      boundBone {boneName=dyn_lightbulb_9}
      boxExtents=<22 22 22>
      constraintSetup {twistAxis=AxisZ, angularLimitsMax=<1 15 15>, angularLimitsMin=<-1 -15 -15>, angularTarget=<0 0 0>}
      localJointOffset=<0 0 0.1>
     }]
    simSpaceSettings {maxAngularAcceleration=150, maxAngularVelocity=150, simSpaceAngularAlpha=1}
   }
   n5 = Anim Dynamics - Bone: dyn_lightbulb_8!(SimulationSpace:AnimPhysSimSpaceType Enum=World, ComponentPose=@n4, Alpha=1, ExternalForce="0,0,0") -> (Pose) {
    angularDampingOverride=0.1
    angularSpringConstant=500
    bAngularSpring=true
    bOverrideLinearDamping=true
    bUseGravityOverride=true
    linearDampingOverride=1
    simulationSpace=World
    alphaBoolBlend {blendOption=Linear}
    boundBone {boneName=dyn_lightbulb_8}
    componentAppliedLinearAccClamp=<1000 1000 1000>
    componentLinearAccScale=<1 1 1>
    componentLinearVelScale=<1 1 1>
    gravityOverride {z=-980}
    physicsBodyDefinitions [{
      boundBone {boneName=dyn_lightbulb_8}
      boxExtents=<22 22 22>
      constraintSetup {twistAxis=AxisZ, angularLimitsMax=<1 15 15>, angularLimitsMin=<-1 -15 -15>, angularTarget=<0 0 0>}
      localJointOffset=<0 0 0.1>
     }]
    simSpaceSettings {maxAngularAcceleration=150, maxAngularVelocity=150, simSpaceAngularAlpha=1}
   }
   n6 = Anim Dynamics - Bone: dyn_lightbulb_7!(SimulationSpace:AnimPhysSimSpaceType Enum=World, ComponentPose=@n5, Alpha=1, ExternalForce="0,0,0") -> (Pose) {
    angularDampingOverride=0.1
    angularSpringConstant=500
    bAngularSpring=true
    bOverrideLinearDamping=true
    bUseGravityOverride=true
    linearDampingOverride=1
    simulationSpace=World
    alphaBoolBlend {blendOption=Linear}
    boundBone {boneName=dyn_lightbulb_7}
    componentAppliedLinearAccClamp=<1000 1000 1000>
    componentLinearAccScale=<1 1 1>
    componentLinearVelScale=<1 1 1>
    gravityOverride {z=-980}
    physicsBodyDefinitions [{
      boundBone {boneName=dyn_lightbulb_7}
      boxExtents=<22 22 22>
      constraintSetup {twistAxis=AxisZ, angularLimitsMax=<1 15 15>, angularLimitsMin=<-1 -15 -15>, angularTarget=<0 0 0>}
      localJointOffset=<0 0 0.1>
     }]
    simSpaceSettings {maxAngularAcceleration=150, maxAngularVelocity=150, simSpaceAngularAlpha=1}
   }
   n7 = Anim Dynamics - Bone: dyn_lightbulb_6!(SimulationSpace:AnimPhysSimSpaceType Enum=World, ComponentPose=@n6, Alpha=1, ExternalForce="0,0,0") -> (Pose) {
    angularDampingOverride=0.1
    angularSpringConstant=500
    bAngularSpring=true
    bOverrideLinearDamping=true
    bUseGravityOverride=true
    linearDampingOverride=1
    simulationSpace=World
    alphaBoolBlend {blendOption=Linear}
    boundBone {boneName=dyn_lightbulb_6}
    componentAppliedLinearAccClamp=<1000 1000 1000>
    componentLinearAccScale=<1 1 1>
    componentLinearVelScale=<1 1 1>
    gravityOverride {z=-980}
    physicsBodyDefinitions [{
      boundBone {boneName=dyn_lightbulb_6}
      boxExtents=<22 22 22>
      constraintSetup {twistAxis=AxisZ, angularLimitsMax=<1 15 15>, angularLimitsMin=<-1 -15 -15>, angularTarget=<0 0 0>}
      localJointOffset=<0 0 0.1>
     }]
    simSpaceSettings {maxAngularAcceleration=150, maxAngularVelocity=150, simSpaceAngularAlpha=1}
   }
   n8 = Anim Dynamics - Bone: dyn_lightbulb_5!(SimulationSpace:AnimPhysSimSpaceType Enum=World, ComponentPose=@n7, Alpha=1, ExternalForce="0,0,0") -> (Pose) {
    angularDampingOverride=0.1
    angularSpringConstant=500
    bAngularSpring=true
    bOverrideLinearDamping=true
    bUseGravityOverride=true
    linearDampingOverride=1
    simulationSpace=World
    alphaBoolBlend {blendOption=Linear}
    boundBone {boneName=dyn_lightbulb_5}
    componentAppliedLinearAccClamp=<1000 1000 1000>
    componentLinearAccScale=<1 1 1>
    componentLinearVelScale=<1 1 1>
    gravityOverride {z=-980}
    physicsBodyDefinitions [{
      boundBone {boneName=dyn_lightbulb_5}
      boxExtents=<22 22 22>
      constraintSetup {twistAxis=AxisZ, angularLimitsMax=<1 15 15>, angularLimitsMin=<-1 -15 -15>, angularTarget=<0 0 0>}
      localJointOffset=<0 0 0.1>
     }]
    simSpaceSettings {maxAngularAcceleration=150, maxAngularVelocity=150, simSpaceAngularAlpha=1}
   }
   n9 = Anim Dynamics - Bone: dyn_lightbulb_4!(SimulationSpace:AnimPhysSimSpaceType Enum=World, ComponentPose=@n8, Alpha=1, ExternalForce="0,0,0") -> (Pose) {
    angularDampingOverride=0.1
    angularSpringConstant=500
    bAngularSpring=true
    bOverrideLinearDamping=true
    bUseGravityOverride=true
    linearDampingOverride=1
    simulationSpace=World
    alphaBoolBlend {blendOption=Linear}
    boundBone {boneName=dyn_lightbulb_4}
    componentAppliedLinearAccClamp=<1000 1000 1000>
    componentLinearAccScale=<1 1 1>
    componentLinearVelScale=<1 1 1>
    gravityOverride {z=-980}
    physicsBodyDefinitions [{
      boundBone {boneName=dyn_lightbulb_4}
      boxExtents=<22 22 22>
      constraintSetup {twistAxis=AxisZ, angularLimitsMax=<1 15 15>, angularLimitsMin=<-1 -15 -15>, angularTarget=<0 0 0>}
      localJointOffset=<0 0 0.1>
     }]
    simSpaceSettings {maxAngularAcceleration=150, maxAngularVelocity=150, simSpaceAngularAlpha=1}
   }
   n10 = Anim Dynamics - Bone: dyn_lightbulb_3!(SimulationSpace:AnimPhysSimSpaceType Enum=World, ComponentPose=@n9, Alpha=1, ExternalForce="0,0,0") -> (Pose) {
    angularDampingOverride=0.1
    angularSpringConstant=500
    bAngularSpring=true
    bOverrideLinearDamping=true
    bUseGravityOverride=true
    linearDampingOverride=1
    simulationSpace=World
    alphaBoolBlend {blendOption=Linear}
    boundBone {boneName=dyn_lightbulb_3}
    componentAppliedLinearAccClamp=<1000 1000 1000>
    componentLinearAccScale=<1 1 1>
    componentLinearVelScale=<1 1 1>
    gravityOverride {z=-980}
    physicsBodyDefinitions [{
      boundBone {boneName=dyn_lightbulb_3}
      boxExtents=<22 22 22>
      constraintSetup {twistAxis=AxisZ, angularLimitsMax=<1 15 15>, angularLimitsMin=<-1 -15 -15>, angularTarget=<0 0 0>}
      localJointOffset=<0 0 0.1>
     }]
    simSpaceSettings {maxAngularAcceleration=150, maxAngularVelocity=150, simSpaceAngularAlpha=1}
   }
   n11 = Anim Dynamics - Bone: dyn_lightbulb_1!(SimulationSpace:AnimPhysSimSpaceType Enum=World, ComponentPose=@n10, Alpha=1, ExternalForce="0,0,0") -> (Pose) {
    angularDampingOverride=0.1
    angularSpringConstant=500
    bAngularSpring=true
    bOverrideLinearDamping=true
    bUseGravityOverride=true
    linearDampingOverride=1
    simulationSpace=World
    alphaBoolBlend {blendOption=Linear}
    boundBone {boneName=dyn_lightbulb_1}
    componentAppliedLinearAccClamp=<1000 1000 1000>
    componentLinearAccScale=<1 1 1>
    componentLinearVelScale=<1 1 1>
    gravityOverride {z=-980}
    physicsBodyDefinitions [{
      boundBone {boneName=dyn_lightbulb_1}
      boxExtents=<22 22 22>
      constraintSetup {twistAxis=AxisZ, angularLimitsMax=<1 15 15>, angularLimitsMin=<-1 -15 -15>, angularTarget=<0 0 0>}
      localJointOffset=<0 0 0.1>
     }]
    simSpaceSettings {maxAngularAcceleration=150, maxAngularVelocity=150, simSpaceAngularAlpha=1}
   }
   n12 = Anim Dynamics - Bone: dyn_lightbulb_0!(SimulationSpace:AnimPhysSimSpaceType Enum=World, ComponentPose=@n11, Alpha=1, ExternalForce="0,0,0") -> (Pose) {
    angularDampingOverride=0.1
    angularSpringConstant=500
    bAngularSpring=true
    bOverrideLinearDamping=true
    bUseGravityOverride=true
    linearDampingOverride=1
    simulationSpace=World
    alphaBoolBlend {blendOption=Linear}
    boundBone {boneName=dyn_lightbulb_0}
    componentAppliedLinearAccClamp=<1000 1000 1000>
    componentLinearAccScale=<1 1 1>
    componentLinearVelScale=<1 1 1>
    gravityOverride {z=-980}
    physicsBodyDefinitions [{
      boundBone {boneName=dyn_lightbulb_0}
      boxExtents=<22 22 22>
      constraintSetup {twistAxis=AxisZ, angularLimitsMax=<0 15 15>, angularLimitsMin=<0 -15 -15>, angularTarget=<0 0 0>}
      localJointOffset=<0 0 0.1>
     }]
    simSpaceSettings {maxAngularAcceleration=150, maxAngularVelocity=150, simSpaceAngularAlpha=1}
   }
   n13 = Anim Dynamics - Bone: wire_1!(SimulationSpace:AnimPhysSimSpaceType Enum=Component, ComponentPose=@n12, Alpha=1, ExternalForce="0,0,0") -> (Pose) {
    angularDampingOverride=0.1
    angularSpringConstant=500
    bAngularSpring=true
    bOverrideLinearDamping=true
    bUseGravityOverride=true
    linearDampingOverride=1
    simulationSpace=World
    alphaBoolBlend {blendOption=Linear}
    boundBone {boneName=wire_1}
    componentAppliedLinearAccClamp=<3000 3000 3000>
    gravityOverride {z=-980}
    physicsBodyDefinitions [{
      boundBone {boneName=wire_1}
      boxExtents=<22 22 22>
      constraintSetup {angularLimitsMax=<0 6 0>, angularLimitsMin=<0 -6 0>, angularTarget=<1 0 0>}
      localJointOffset=<0.6 0 0>
     }]
    simSpaceSettings {maxAngularAcceleration=150, maxAngularVelocity=150, simSpaceAngularAlpha=0.2688}
   }
   n14 = Anim Dynamics - Bone: wire_2!(SimulationSpace:AnimPhysSimSpaceType Enum=Component, ComponentPose=@n13, Alpha=1, ExternalForce="0,0,0") -> (Pose) {
    angularDampingOverride=0.1
    angularSpringConstant=500
    bAngularSpring=true
    bOverrideLinearDamping=true
    bUseGravityOverride=true
    linearDampingOverride=1
    simulationSpace=World
    alphaBoolBlend {blendOption=Linear}
    boundBone {boneName=wire_2}
    componentAppliedLinearAccClamp=<3000 3000 3000>
    gravityOverride {z=-980}
    physicsBodyDefinitions [{
      boundBone {boneName=wire_2}
      boxExtents=<14 14 14>
      constraintSetup {angularLimitsMax=<0 0 0>, angularLimitsMin=<0 0 0>, angularTarget=<1 0 0>}
      localJointOffset=<0.6 0 0>
     }]
    simSpaceSettings {maxAngularAcceleration=150, maxAngularVelocity=150, simSpaceAngularAlpha=0.6}
   }
   n15 = Anim Dynamics - Bone: wire_3!(SimulationSpace:AnimPhysSimSpaceType Enum=Component, ComponentPose=@n14, Alpha=1, ExternalForce="0,0,0") -> (Pose) {
    angularDampingOverride=0.1
    angularSpringConstant=500
    bAngularSpring=true
    bOverrideLinearDamping=true
    bUseGravityOverride=true
    linearDampingOverride=1
    simulationSpace=World
    alphaBoolBlend {blendOption=Linear}
    boundBone {boneName=wire_3}
    componentAppliedLinearAccClamp=<3000 3000 3000>
    componentLinearAccScale=<1 1 1>
    componentLinearVelScale=<1 1 1>
    gravityOverride {z=-980}
    physicsBodyDefinitions [{
      boundBone {boneName=wire_3}
      boxExtents=<22 22 22>
      constraintSetup {angularLimitsMax=<4 15 15>, angularLimitsMin=<-4 -15 -15>, angularTarget=<1 0 0>}
      localJointOffset=<0.6 0 0>
     }]
    simSpaceSettings {maxAngularAcceleration=150, maxAngularVelocity=150, simSpaceAngularAlpha=0.6}
   }
   Output Pose(Result=Component To Local(ComponentPose=@n15)) {layerGroup=None}
  ]}
 EventGraph {nodes [
   Event BlueprintUpdateAnimation()
   > n1 = Branch(Condition=Get bIsInFrontEnd())
   > Set SimlationSpace(SimlationSpace:AnimPhysSimSpaceType Enum=World) -> (Output_Get:AnimPhysSimSpaceType Enum)
   Set SimlationSpace(execute=@n1.else, SimlationSpace=Component)
   TryGetPawnOwner()
  ]}
]
variables [SimlationSpace {type=AnimPhysSimSpaceType}]
